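; B_DoomSmoke is a Doom Castle smoke effect actor with top and bottom post-process volumes, each pairing a box with a mesh.
; Its construction script disables actor collision and writes each box's extent, rotation axes, fade distance and cylinder radius into that mesh's default custom primitive data for its material.
; The top mesh also receives hardcoded vectors, and its fade floats later overwrite part of the one at index 20.
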(fn ConstructionScript
  (bind _postprocessarea_top (Variables|Default|GetPostProcessArea_Top))
  (bind _returnvalue (Transformation|GetWorldRotation _postprocessarea_top))
  (bind _postprocessarea_bottom (Variables|Default|GetPostProcessArea_Bottom))
  (bind _returnvalue_2 (Transformation|GetWorldRotation _postprocessarea_bottom))
  (bind _postprocessmesh_top (Variables|Default|GetPostProcessMesh_Top))
  (bind _postprocessmesh_bottom (Variables|Default|GetPostProcessMesh_Bottom))
  (bind _postprocessmesh_bottom_2 (Variables|Default|GetPostProcessMesh_Bottom))
  (bind _postprocessmesh_top_2 (Variables|Default|GetPostProcessMesh_Top))
  ; Temp
  (Collision|SetActorEnableCollision false)
  (Utilities|FlowControl|Sequence
    (:then_0
      (Rendering|Material|SetDefaultCustomPrimitiveDataVector3 _postprocessmesh_top 20 "0.700000,0.700000,0.399000")
      (Rendering|Material|SetDefaultCustomPrimitiveDataVector3 _postprocessmesh_top 24 "0.010000,0.010000,-0.010100")
      (Rendering|Material|SetDefaultCustomPrimitiveDataVector3 _postprocessmesh_top 28 "1.200000,1.200000,1.200000")
      (Rendering|Material|SetDefaultCustomPrimitiveDataVector3 _postprocessmesh_top 32 "0.000000,0.000000,-0.002300")
      (Rendering|Material|SetDefaultCustomPrimitiveDataVector3 _postprocessmesh_top 16 (Math|Conversions|ToVector(LinearColor) "(R=0.100000,G=0.100000,B=0.100000,A=1.000000)")))
    (:then_1
      (Rendering|Material|SetDefaultCustomPrimitiveDataVector3 _postprocessmesh_top_2 0 (Components|Box|GetScaledBoxExtent (Variables|Default|GetPostProcessArea_Top)))
      (Rendering|Material|SetDefaultCustomPrimitiveDataVector3 _postprocessmesh_bottom_2 4 (Math|Vector|GetForwardVector _returnvalue_2))
      (Rendering|Material|SetDefaultCustomPrimitiveDataVector3 _postprocessmesh_bottom_2 8 (Math|Vector|GetRightVector _returnvalue_2))
      (Rendering|Material|SetDefaultCustomPrimitiveDataVector3 _postprocessmesh_bottom_2 12 (Math|Vector|GetUpVector _returnvalue_2))
      (Rendering|Material|SetDefaultCustomPrimitiveDataFloat _postprocessmesh_top_2 20 (Variables|Default|GetFadeDistance_Top))
      (Rendering|Material|SetDefaultCustomPrimitiveDataFloat _postprocessmesh_top_2 21 (Variables|Default|GetFadeCylinderRadius_Top))
      (Rendering|Material|SetDefaultCustomPrimitiveDataVector3 _postprocessmesh_top_2 4 (Math|Vector|GetForwardVector _returnvalue))
      (Rendering|Material|SetDefaultCustomPrimitiveDataVector3 _postprocessmesh_top_2 8 (Math|Vector|GetRightVector _returnvalue))
      (Rendering|Material|SetDefaultCustomPrimitiveDataVector3 _postprocessmesh_top_2 12 (Math|Vector|GetUpVector _returnvalue)))
    (:then_2
      (Rendering|Material|SetDefaultCustomPrimitiveDataVector3 _postprocessmesh_bottom 0 (Components|Box|GetScaledBoxExtent (Variables|Default|GetPostProcessArea_Bottom)))
      (Rendering|Material|SetDefaultCustomPrimitiveDataFloat _postprocessmesh_bottom 20 (Variables|Default|GetFadeDistance_Bottom))
      (Rendering|Material|SetDefaultCustomPrimitiveDataFloat _postprocessmesh_bottom 21 (Variables|Default|GetFadeCylinderRadius_Bottom)))))
