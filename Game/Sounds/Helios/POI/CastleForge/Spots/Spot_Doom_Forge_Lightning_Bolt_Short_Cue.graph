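# Sound cue for a short lightning bolt spot sound at the Castle Forge (Doom Forge) POI.
# One of ten short blast waves plays at random with slight pitch variation.
# That signal feeds three attenuation layers (close no-focus, close focus, focus), mixed with the focus layer loudest, under the first two quality slots.
# Concurrency is capped at three voices, with sends to exterior ambient bus and reverb.
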SoundCue /Game/Sounds/Helios/POI/CastleForge/Spots/Spot_Doom_Forge_Lightning_Bolt_Short_Cue
root=QualityLevel_0
nodes [
 {
  id=Attenuation_0
  children [Modulator_0]
  properties {attenuationSettings=&/Doom_Forge_LightningBolts_Focus_Att}
 }
 {
  id=Attenuation_1
  children [Modulator_0]
  properties {attenuationSettings=&/Doom_Forge_LightningBolts_Close_NoFocus_Att}
 }
 {
  id=Attenuation_2
  children [Modulator_0]
  properties {attenuationSettings=&/Doom_Forge_LightningBolts_Close_Focus_Att}
 }
 {id=Mixer_0, children [Attenuation_1, Attenuation_2, Attenuation_0], properties {inputVolume [0.33,0.4,0.8]}}
 {id=Modulator_0, children [Random_0], properties {pitchMax=1.1, pitchMin=0.9}}
 {id=QualityLevel_0, children [Mixer_0,Mixer_0,null,null]}
 {
  id=Random_0
  children [WavePlayer_0,WavePlayer_1,WavePlayer_2,WavePlayer_3,WavePlayer_4,WavePlayer_5,WavePlayer_6,WavePlayer_7,WavePlayer_8,WavePlayer_9]
  properties {weights [1,1,1,1,1,1,1,1,1,1]}
 }
 {id=WavePlayer_0, properties {soundWaveAssetPtr=&/Spot_Lightning_Blast_Single_Short_01}}
 {id=WavePlayer_1, properties {soundWaveAssetPtr=&/Spot_Lightning_Blast_Single_Short_02}}
 {id=WavePlayer_2, properties {soundWaveAssetPtr=&/Spot_Lightning_Blast_Single_Short_03}}
 {id=WavePlayer_3, properties {soundWaveAssetPtr=&/Spot_Lightning_Blast_Single_Short_04}}
 {id=WavePlayer_4, properties {soundWaveAssetPtr=&/Spot_Lightning_Blast_Single_Short_05}}
 {id=WavePlayer_5, properties {soundWaveAssetPtr=&/Spot_Lightning_Blast_Single_Short_06}}
 {id=WavePlayer_6, properties {soundWaveAssetPtr=&/Spot_Lightning_Blast_Single_Short_07}}
 {id=WavePlayer_7, properties {soundWaveAssetPtr=&/Spot_Lightning_Blast_Single_Short_08}}
 {id=WavePlayer_8, properties {soundWaveAssetPtr=&/Spot_Lightning_Blast_Single_Short_09}}
 {id=WavePlayer_9, properties {soundWaveAssetPtr=&/Spot_Lightning_Blast_Single_Short_10}}
]
properties {
 bOverrideConcurrency=true
 maxDistance=43000
 soundClassObject=/Game/Sounds/FortSoundClassesAndModes/Fort_Storms_PSM
 volumeMultiplier=0.5
 attenuationOverrides {
  bAttenuateWithLPF=true
  bEnableLogFrequencyScaling=true
  distanceAlgorithm=Logarithmic
  falloffDistance=10000
  lPFFrequencyAtMax=197.5057
  lPFFrequencyAtMin=20511.4102
  lPFRadiusMax=3789.3618
  lPFRadiusMin=1822.3112
  stereoSpread=2000
  attenuationShapeExtents {x=0}
 }
 busSends [{audioBus=null, minSendLevel=1, sendLevel=1, soundSourceBus=&AmbientAudio_Exterior_SourceBus}]
 concurrencyOverrides {maxCount=3, voiceStealReleaseTime=2, volumeScaleAttackTime=0.5}
 soundSubmixSends [{minSendLevel=0.6, soundSubmix=&Fortnite_Reverb_Default}]
}
